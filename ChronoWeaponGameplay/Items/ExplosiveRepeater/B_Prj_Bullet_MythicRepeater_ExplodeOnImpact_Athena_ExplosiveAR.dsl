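; Blueprint for an explode-on-impact bullet projectile of the Mythic Repeater explosive weapon.
; Its construction script sets the tracer mesh's max draw distance from a draw distance variable.
; An empty function named NewFunction is also present, and no impact or explosion logic appears in these graphs.
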(fn NewFunction)

(fn ConstructionScript
  (LOD|SetMaxDrawDistance (Variables|Default|GetTracerMesh_0) (Variables|Default|GetDrawDistance_0)))
