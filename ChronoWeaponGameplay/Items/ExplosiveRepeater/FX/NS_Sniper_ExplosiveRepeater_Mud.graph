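# Mud impact effect for the Explosive Repeater sniper: a one-shot Niagara burst of splashes, droplets and debris, plus a fading decal and a mud explosion sound.
# System surface flags (mainly SpawnMud) gate most burst counts, and User.Global Scale drives much of the size and velocity.
# CPU and GPU emitters are mixed, with one CPU emitter exporting collision normal and material index to Blueprint, then killing the collided particle.
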NiagaraSystem /ChronoWeaponGameplay/Items/ExplosiveRepeater/FX/NS_Sniper_ExplosiveRepeater_Mud
dependencies {
 usedDataInterfaces [
  &NiagaraDataInterfaceCurve
  &NiagaraDataInterfaceArrayFloat
  &NiagaraDataInterfaceSpriteRendererInfo
  &NiagaraDataInterfaceArrayFloat3
  /Script/Niagara.NiagaraDataInterfaceExport
  &NiagaraDataInterfaceAudioPlayer
 ]
 usedDynamicInputs [
  &RotateVectorByQuaternion
  &self:ScratchDynamicInput_01
 ]
 usedModules [
  &SystemState
  &self:SystemSpawnScript.SystemScriptSource.SystemScriptGraph.NiagaraNodeAssignment_1.SetVariables_#1
  &self:SetPeeloutFXArrayIndex
 ]
}
emitters {
 Collided_Mud_Splash001 {
  emitterUpdateScript [
   &EmitterState()
   &SpawnBurst_Instantaneous(
    Spawn Count=&Multiply_Int(
     A=&RandomRangeInt(Minimum=50, Maximum=70)
     B=&MakeCustomIntFromBool(Bool=$System.SpawnMud, True INT=1, False INT=0)
    )
    Spawn Time=0.03
    Use Spawn Probability=false
   )
  ]
  particleSpawnScript [
   &InitializeParticle(
    Color=<rgba 0.1042 0.0297 0.011 1>
    Lifetime Max=0.5
    Lifetime Min=0.3
    Lifetime Mode=::NewEnumerator1
    Sprite Rotation Mode=::NewEnumerator2
    Sprite Size Mode=::NewEnumerator1
    Sprite UV Mode=::NewEnumerator3
    Uniform Sprite Size Max=45
    Uniform Sprite Size Min=35
    UsePositionOffset=true
    Write Material Random=true
   )
   &SubUVAnimation(
    "Sprite Renderer "=DataInterface() {propertyValues {SpriteRenderer=&self:Collided_Mud_Splash001_3.NiagaraSpriteRendererProperties_1}}
    SubUV Animation Mode:ENiagara_SubUVLookupModeV2=::NewEnumerator2
    UseEndFrame=false
    UseStartFrame=false
   )
   &ShapeLocation(MaskSpawnGroup=false, Override Seed=false, Sphere Radius=16, UseOwnerScale=true)
   &AddVelocity(
    Velocity=$Particles.ShapeLocation.ShapeNormal
    Velocity Speed Scale=&Multiply_Float(A=2000, B=$User.Global Scale)
   ) {module=AddVelocity002}
   &CameraOffset(Camera Offset Amount=15)
  ]
  particleUpdateScript [
   &ParticleState()
   &GravityForce(Gravity=&MakeVector(X=0, Y=0, Z=&Multiply_Float(A=-980, B=1)))
   &InheritVelocity(
    Inherited Velocity Amount Scale=&MakeVector(X=0.1, Y=0.1, Z=0.1)
    Speed Limit=true
    UseSpeedThreshold=false
   ) {enabled=false}
   &Drag(Drag=0.5, Use Linear Drag=true, Use Rotational Drag=false)
   &ScaleColor(
    Scale Alpha=&FloatFromCurve(FloatCurve=DataInterface() {propertyValues {Curve {keys [
         {tangentMode=RCTM_User, time=0, value=1}
         {arriveTangent=-0.1934, interpMode=RCIM_Cubic, leaveTangent=-0.1934, tangentMode=RCTM_User, time=0.4, value=1}
         {arriveTangent=-0.2877, interpMode=RCIM_Cubic, leaveTangent=-0.2877, tangentMode=RCTM_User, time=1, value=0}
        ]}}})
    ScaleA=true
    ScaleRGB=true
   )
   &ScaleSpriteSize(
    Uniform Curve Scale=1.5
    Uniform Curve Sprite Scale=DataInterface() {propertyValues {Curve {keys [
        {arriveTangent=4.1785, interpMode=RCIM_Cubic, leaveTangent=4.1785, tangentMode=RCTM_User, time=0, value=0}
        {arriveTangent=1.1179, interpMode=RCIM_Cubic, leaveTangent=1.1179, tangentMode=RCTM_User, time=0.4, value=0.75}
        {arriveTangent=0.0224, interpMode=RCIM_Cubic, leaveTangent=0.0224, tangentMode=RCTM_User, time=1, value=1}
       ]}}}
   )
   &SolveForcesAndVelocity(Clamp Velocity=false, Limit Acceleration=false)
  ]
  rendererClasses [&NiagaraSpriteRendererProperties]
  renderers [{rendererClass=&NiagaraSpriteRendererProperties}]
 }
 Collision_Send_BP {
  emitterUpdateScript [
   &EmitterState(
    Enable Distance Culling=true
    MaxDistance=10000
    Module.bUseMaxDistance=true
    Module.bUseMinDistance=false
    Scalability Mode=::NewEnumerator1
   )
   &SpawnRate(Use Spawn Probability=false) {enabled=false}
   &SpawnBurst_Instantaneous(
    Spawn Count=&MakeIntFromFloat(Float=&Multiply_Float(A=6, B=$User.Global Scale))
    Spawn Group:NiagaraInt32=$System.NewGroup
    Spawn Time=$System.Delay
    Use Spawn Probability=false
    UseLoopCountLimit=false
   )
  ]
  particleSpawnScript [
   /SetVariables(Particles.VisibilityTag=$Emitter.SpawnGroup) {isSetParameters=true}
   &InitializeParticle(
    Color=<rgba 1.3728 3.0599 5 1>
    Lifetime Min=1.5
    Lifetime Mode=::NewEnumerator1
    Sprite Size Mode=::NewEnumerator1
    Uniform Sprite Size Max=38
    Uniform Sprite Size Min=32
    UsePositionOffset=false
    Write Material Random=true
   )
   &ShapeLocation(
    Handle Radius:NiagaraFloat=&Multiply_Float(A=50, B=$User.Global Scale)
    Large Radius=&Multiply_Float(A=200, B=$User.Global Scale)
    MaskSpawnGroup=false
    Non Uniform Scale=<1 1 0.1>
    Offset=<0 0 30>
    Offset Mode=::NewEnumerator0
    Override Seed=false
    Shape Primitive=::NewEnumerator3
    UseOwnerScale=true
   )
   &ShapeLocation(
    Apply To Particle Position=::NewEnumerator1
    MaskSpawnGroup=false
    Non Uniform Scale=<1 1 0.1>
    Offset=<5 0 30>
    Offset Coordinate Space=::World
    Offset Mode=::NewEnumerator0
    Override Seed=false
    Rotation Coordinate Space=::World
    Rotation Mode=::NewEnumerator1
    Shape Origin=&VectorFromFloat(Value=0)
    Sphere Surface Distribution=1
    UseOwnerScale=true
   ) {enabled=false, module=ShapeLocation001}
   &SkeletalMeshLocation(
    Bone / Socket Index=8
    Bone Sampling Mode=::NewEnumerator4
    Interpolate Mesh Sampling=false
    MaskSpawnGroup=false
    Sampled Position Offset=$Particles.ShapeLocation001.ShapePosition
    Skeletal Mesh=$User.SkeletalMesh
    Use Sampled Mesh Attribute as Ribbon ID=false
   ) {enabled=false}
   &AddVelocity(
    Additional Quaternion Rotation=false
    Rotation Mode=::NewEnumerator3
    Velocity=$Particles.ShapeLocation001.ShapeNormal
    Velocity Speed Scale=600
   ) {enabled=false, module=AddVelocity001}
   &AddVelocity(
    Module.Use Falloff Distance=false
    Origin Offset=&RandomRangeVector(Minimum=<0 0 -250>, Maximum=<0 0 -150>)
    Override Seed=false
    Velocity Mode=::NewEnumerator1
    Velocity Speed=&Multiply_Float(A=&RandomRangeFloat(Minimum=700, Maximum=1000), B=$User.Global Scale)
   )
   &ApplyInitialForces(Apply Rotational Force to Rotational Velocity=false) {enabled=false}
   /SetVariables(Particles.SpriteSize=&Multiply_Vector2D_ByFloat(Float=$User.Global Scale, Vector2D=$Particles.SpriteSize)) {isSetParameters=true, module=SetVariables_#2}
  ]
  particleUpdateScript [
   &ParticleState()
   &GravityForce(Gravity=&MakeCustomVectorFromBool(Bool=$Particles.HasCollided, True Vector=<0 0 -500>, False Vector=<1 0 -1200>))
   &Collision(
    Age Colliding Particles=false
    Friction=1
    Friction During a Bounce=1
    Max CPU Trace Length=100
    Module.Control Rest State Via Surface Normal=false
    Module.EnableMaxCollisionCount=false
    Module.Randomize Collision Normal Vector=false
    Restitution=0
   ) {enabled=false, module=Collision001}
   &Collision(
    Age Colliding Particles=false
    Collision Enabled=&SetBoolByFloatComparison(A=1, B=&RandomRangeFloat(Minimum=0, Maximum=1))
    Module.Control Rest State Via Surface Normal=false
    Module.EnableMaxCollisionCount=false
    Module.Randomize Collision Normal Vector=false
    Restitution=0
   )
   &InheritSourceMovement(
    Applied Position Scale=&VectorFromFloat(Value=&MakeCustomFloatFromBool(Bool=$Particles.HasCollided, True Float=0, False Float=0.9))
    ApplyForce=false
    ApplyPosition=true
    ApplyPositionFalloff=false
   )
   &SolveForcesAndVelocity(Clamp Velocity=false, Limit Acceleration=false)
   &Timeline(
    Additional Output:ENiagara_TimelineOutput=::NewEnumerator1
    Additional Output Mode:ENiagara_TimelineOutputMode=::NewEnumerator1
    Float Curve=DataInterface() {propertyValues {Curve {keys [
        {arriveTangent=9.0725, interpMode=RCIM_Cubic, leaveTangent=9.0725, tangentMode=RCTM_User, time=0, value=0}
        {arriveTangent=-2.5249, interpMode=RCIM_Cubic, leaveTangent=-2.5249, tangentMode=RCTM_User, time=0.1408, value=0.9847}
        {arriveTangent=-0.0405, interpMode=RCIM_Cubic, leaveTangent=-0.0405, tangentMode=RCTM_User, time=1, value=0}
       ]}}}
    Instantaneous Rewind=true
    Play=$Particles.HasCollided
   )
   &DynamicMaterialParameters(
    Index 0 Param 1=0.5
    Index 0 Param 2=5
    Index 0 Param 3=600
    Index 0 Param 4=8
    Index 1 Param 1=&FloatFromCurve(
     FloatCurve=DataInterface() {propertyValues {Curve {keys [
         {arriveTangent=-3.7492, interpMode=RCIM_Cubic, leaveTangent=-3.7492, tangentMode=RCTM_User, time=0, value=1}
         {arriveTangent=-0.7477, interpMode=RCIM_Cubic, leaveTangent=-0.7477, tangentMode=RCTM_User, time=0.3, value=0.3}
         {interpMode=RCIM_Cubic, leaveTangent=-1.9286, time=1, value=0}
        ]}}}
     CurveIndex=$Particles.Timeline.NormalizedTime
     Scale Curve=&RandomRangeFloat(Minimum=2, Maximum=2.25)
    )
    Module.Param0WriteEnabled=true
    Module.Param0WriteEnabled1=true
    Module.Param1WriteEnabled=true
    Module.Param1WriteEnabled1=false
    Module.Param2WriteEnabled=true
    Module.Param2WriteEnabled1=false
    Module.Param3WriteEnabled=true
    Module.Param3WriteEnabled1=false
    Write Parameter Index 1=true
   )
   &ScaleColor(Scale RGB=<0 0 0>, ScaleA=false, ScaleRGB=true)
   &ScaleSpriteSize(Uniform Curve Sprite Scale=DataInterface() {propertyValues {Curve {keys [
        {interpMode=RCIM_Cubic, tangentMode=RCTM_User, time=0.7, value=1}
        {arriveTangent=-6.3247, interpMode=RCIM_Cubic, leaveTangent=-6.3247, tangentMode=RCTM_User, time=1, value=0}
       ]}}})
   /Niagara/Modules/ExportParticleData/ExportParticleDataToBlueprint(
    Condition To Export Data:NiagaraBool=$Transient.CollisionValid
    Export Particle Data Interface=DataInterface() {propertyValues {
      GPUAllocationFixedSize=64
      GPUAllocationMode=FixedSize
      GPUAllocationPerParticleSize=1
      CallbackHandlerParameter {parameter {name=User.Export particle data, typeDefHandle {registeredTypeIndex=1}, varData []}}
     }}
    "Float To Send (As Struct Size)":NiagaraFloat=&MakeFloatFromInt(Int=$Particles.Collision.CollisionQueryAndResponse.Collision1MaterialIndex)
    UseDelayTimer=false
    "Vector To Send (As Struct Velocity)":Vector3f=$Transient.CollisionNormal
   )
   &KillParticles(Kill Particles=$Output.Collision.CollisionValid)
  ]
  rendererClasses [&NiagaraSpriteRendererProperties]
  renderers [{rendererClass=&NiagaraSpriteRendererProperties}]
 }
 Decal {
  emitterUpdateScript [
   &EmitterState(Life Cycle Mode=::NewEnumerator1, Loop Behavior=::NewEnumerator1, Loop Duration=2, UseLoopDelay=false)
   &SpawnBurst_Instantaneous(Use Spawn Probability=false, UseLoopCountLimit=false)
  ]
  particleSpawnScript [
   &InitializeParticle(
    Lifetime=&Multiply_Float(A=1.8, B=$User.Global Scale)
    Sprite Size Mode=::NewEnumerator3
    Uniform Sprite Size=&Multiply_Float(A=100, B=$User.Global Scale)
    UsePositionOffset=false
    Write Material Random=true
   )
   /SetVariables(Particles.DecalSize:Vector3f=<200 400 400>) {isSetParameters=true}
  ]
  particleUpdateScript [
   &ParticleState()
   &ScaleColor(
    Scale Alpha=&FloatFromCurve(FloatCurve=DataInterface() {propertyValues {Curve {keys [
         {arriveTangent=1.18, interpMode=RCIM_Cubic, leaveTangent=1.18, tangentMode=RCTM_User, time=-0.0046, value=0}
         <at 0.05 1>
         <at 0.6 0.75>
         <at 1.0011 0.002 RCIM_Cubic>
        ]}}})
    ScaleA=true
    ScaleRGB=false
   )
   &SolveForcesAndVelocity(Clamp Velocity=false, Limit Acceleration=false)
   /SetVariables(Particles.DecalFade:NiagaraFloat=&MakeFloatFromLinearColor(LinearColor=$Particles.Color, Channel:ENiagaraLinearColor_Channels=::NewEnumerator3)) {isSetParameters=true}
  ]
  rendererClasses [&NiagaraDecalRendererProperties]
  renderers [{rendererClass=&NiagaraDecalRendererProperties, rendererIndex=0}]
 }
 Drops_Wide {
  simTarget=GPUComputeSim
  emitterUpdateScript [
   &EmitterState(
    Enable Distance Culling=true
    Max Distance Response=::KillAfterParticlesFinish
    Module.bUseMaxDistance=true
    Module.bUseMinDistance=false
    Scalability Mode=::NewEnumerator1
   )
   &SpawnBurst_Instantaneous(
    Spawn Count=&Multiply_Int(
     A=&RandomRangeInt(Minimum=50, Maximum=70)
     B=&MakeCustomIntFromBool(Bool=$System.SpawnMud, True INT=1, False INT=0)
    )
    Spawn Time=0.03
    Use Spawn Probability=false
   )
   &SpawnBurst_Instantaneous(Spawn Group=1, Use Spawn Probability=false, UseLoopCountLimit=false) {module=SpawnBurst_Instantaneous001}
  ]
  particleSpawnScript [
   /SetVariables(Particles.VisibilityTag=$Emitter.SpawnGroup) {isSetParameters=true}
   /Niagara/Modules/Spawn/Initialization/InitializeParticle(
    Color=<rgba 0.3073 0.1546 0.0701 2>
    Lifetime=&UniformRangedFloat(Minimum=0.6, Maximum=1.4)
    Module.Write Color=true
    Module.Write Lifetime=true
    Module.Write Mass=true
    Module.Write Material Random=true
    Module.Write Position=false
    Module.Write Scale=false
    Module.Write SpriteRotation=false
    Module.Write SpriteSize=true
    Position=&SimulationPosition()
    Sprite Size=&UniformRangedVector2D(Minimum=<12 20>, Maximum=<20 32>)
   ) {enabled=false}
   /Niagara/Modules/Spawn/Initialization/V2/InitializeParticle(
    Color=&MaskLinearColorBySpawnGroup(Masked Linear Color=<rgba 1.1 1.2 1.1 1>, Passthrough Linear Color=<rgba 0.2865 0.1376 0.0552 2>, Spawn Group Mask=1)
    Lifetime=&MaskFloatBySpawnGroup(
     Masked Float=&Multiply_Float(A=1.5, B=HlslExpression() {hlslExpression="saturate(pow(User.GlobalScale, 0.5))"})
     Passthrough Float=&RandomRangeFloat(Minimum=0.6, Maximum=1.4)
     Spawn Group Mask=1
    )
    Mesh Scale=&Multiply_VectorByFloat(
     Vector=<1.29 1.29 1.4>
     Float=&Multiply_Float(A=&RandomRangeFloat(Minimum=24, Maximum=26), B=$User.Global Scale)
    )
    Mesh Scale Mode=::NewEnumerator4
    Position Offset=&Multiply_VectorByFloat(Vector=<0 0 70>, Float=$User.Global Scale)
    Sprite Rotation Mode=::NewEnumerator2
    Sprite Size Max=<20 32>
    Sprite Size Min=<12 20>
    Sprite Size Mode=::NewEnumerator2
    Sprite UV Mode=::NewEnumerator3
    UsePositionOffset=true
    Write Material Random=true
   ) {module=InitializeParticle001}
   &AddVelocityInCone(
    Cone Angle=80
    Cone Axis=<0 0 1>
    Module.Use Velocity Falloff On Cone Axis=false
    Velocity Distribution Along Cone Axis=0.75
    Velocity Strength=&MaskFloatBySpawnGroup(
     Masked Float=0
     Passthrough Float=&Multiply_Float(A=&RandomRangeFloat(Minimum=600, Maximum=1500), B=$User.Global Scale)
     Spawn Group Mask=1
    )
   )
   &FN_PerParticleLOD(Enable SpriteSize Distance Scaling=true, Sprite Size Scale Far=<1.5 1.5>)
   &CameraOffset(Camera Offset Amount=10)
   /SetVariables(Particles.SpriteSize=&Multiply_Vector2D_ByFloat(Float=$User.Global Scale, Vector2D=$Particles.SpriteSize)) {isSetParameters=true}
  ]
  particleUpdateScript [
   &ParticleState()
   &AccelerationForce(Acceleration=&MaskVectorBySpawnGroup(Masked Vector=<0 0 -400>, Passthrough Vector=<0 0 -1200>, Spawn Group Mask=1))
   &Drag(Drag=0.1, Use Linear Drag=true, Use Rotational Drag=false) {enabled=false}
   &SolveForcesAndVelocity(Clamp Velocity=false, Limit Acceleration=false)
   &ScaleSpriteSize(Scale Factor=&Vector2DFromFloat(Value=&OneMinusFloat(Float=$Particles.NormalizedAge)))
   &ScaleMeshSize(Scale Factor=&Multiply_VectorByFloat(
     Vector=&MakeVector(
      X=1.5
      Y=1.5
      Z=&FloatFromCurve(FloatCurve=DataInterface() {propertyValues {Curve {keys [
           {tangentMode=RCTM_User, time=0, value=3}
           {arriveTangent=-0.0405, interpMode=RCIM_Cubic, leaveTangent=-0.0405, tangentMode=RCTM_User, time=0.5, value=1.125}
           {arriveTangent=-3.7129, interpMode=RCIM_Cubic, leaveTangent=-3.7129, tangentMode=RCTM_User, time=1, value=0.3}
          ]}}})
     )
     Float=&FloatFromCurve(FloatCurve=DataInterface() {propertyValues {Curve {keys [
          {arriveTangent=1.7414, interpMode=RCIM_Cubic, leaveTangent=1.7414, tangentMode=RCTM_User, time=0.1, value=0.1}
          {arriveTangent=1.466, interpMode=RCIM_Cubic, leaveTangent=1.466, tangentMode=RCTM_User, time=1, value=1}
         ]}}})
    ))
   /Niagara/Modules/Update/SubUV/SubUVAnimation(Random SubUV Lookup Index:NiagaraFloat=$Particles.Initial.MaterialRandom, SubUV Animation Mode=::NewEnumerator1) {enabled=false}
   &ScaleColor(
    Scale Alpha=&FloatFromCurve(FloatCurve=DataInterface() {propertyValues {Curve {keys [<at 0.7 1 RCIM_Cubic>,<at 1.0011 0.002>]}}})
    ScaleA=true
    ScaleRGB=true
   ) {enabled=false}
   &ScaleColor(
    Scale Alpha=&MaskFloatBySpawnGroup(
     Masked Float=&FloatFromCurve(FloatCurve=DataInterface() {propertyValues {Curve {keys [
          {arriveTangent=-5.0374, interpMode=RCIM_Cubic, leaveTangent=-5.0374, tangentMode=RCTM_User, time=0, value=1}
          {arriveTangent=-4.0906, interpMode=RCIM_Cubic, leaveTangent=-4.0906, tangentMode=RCTM_User, time=1, value=-1}
         ]}}})
     Passthrough Float=&FloatFromCurve(FloatCurve=DataInterface() {propertyValues {Curve {keys [<at 0.5 1 RCIM_Cubic>, {arriveTangent=-2.5229, interpMode=RCIM_Cubic, leaveTangent=-2.5229, tangentMode=RCTM_User, time=1, value=0}]}}})
     Spawn Group Mask=1
    )
    Scale RGB=&MaskVectorBySpawnGroup(Masked Vector=<1.2 1.35 1.2>, Passthrough Vector=<1 1 1>, Spawn Group Mask=1)
    ScaleA=true
    ScaleRGB=true
   ) {module=ScaleColor001}
   &Color(Color=&MakeLinearColorFromVectorAndFloat(
     "Vector (RGB)"=<0.8 1 1>
     "Float (Alpha)"=&FloatFromCurve(FloatCurve=DataInterface() {propertyValues {Curve {keys [<at 0.7 1 RCIM_Cubic>,<at 1.0011 0.002>]}}})
    )) {enabled=false}
   &DynamicMaterialParameters(Write Parameter Index 0=false) {enabled=false}
  ]
  rendererClasses [&NiagaraSpriteRendererProperties, &NiagaraMeshRendererProperties]
  renderers [
   {rendererClass=&NiagaraSpriteRendererProperties}
   {rendererClass=&NiagaraMeshRendererProperties, rendererIndex=1}
  ]
 }
 NE_Parent_Mud_DropsCollision {
  emitterUpdateScript [
   &EmitterState(Enable Distance Culling=true, MaxDistance=4000, Module.bUseMaxDistance=true, Module.bUseMinDistance=false, Scalability Mode=::NewEnumerator1)
   &SpawnPerFrame(
    Spawn Count:NiagaraInt32=&MakeCustomIntFromBool(Bool=$System.ValidFrame, True INT=&RandomRangeInt(Minimum=4, Maximum=6), False INT=0)
    Use Spawn Probability=false
   ) {enabled=false}
   &SpawnBurst_Instantaneous(
    Spawn Count=&Multiply_Int(
     A=&RandomRangeInt(Minimum=50, Maximum=70)
     B=&MakeCustomIntFromBool(Bool=$System.SpawnMud, True INT=1, False INT=0)
    )
    Spawn Time=0.03
    Use Spawn Probability=false
   )
  ]
  particleSpawnScript [
   &InitializeParticle(
    Lifetime Max=1
    Lifetime Min=0.6
    Lifetime Mode=::NewEnumerator1
    Sprite Size Mode=::NewEnumerator1
    Uniform Sprite Size Max=5.5
    Uniform Sprite Size Min=2
    UsePositionOffset=false
    Write Material Random=true
   )
   &ShapeLocation(MaskSpawnGroup=false, Non Uniform Scale=<10 1 1>, Override Seed=false, Sphere Radius=16, UseOwnerScale=true)
   &AddVelocity(
    Velocity=$Particles.ShapeLocation.ShapeNormal
    Velocity Speed Scale=&Multiply_Float(A=2400, B=$User.Global Scale)
   )
  ]
  particleUpdateScript [
   &ParticleState()
   &GravityForce(Gravity=&MakeVector(X=0, Y=0, Z=&Multiply_Float(A=-980, B=2))) {module=GravityForce001}
   &InheritVelocity(
    Inherited Velocity Amount Scale=&MakeVector(X=0.1, Y=0.1, Z=0.1)
    Speed Limit=true
    UseSpeedThreshold=false
   )
   &Drag(Use Linear Drag=true, Use Rotational Drag=true)
   &SolveForcesAndVelocity(Clamp Velocity=false, Limit Acceleration=false)
   &DynamicMaterialParameters(
    Index 0 Param 1=$Particles.Timeline.Float
    Index 0 Param 2=5
    Index 0 Param 3=10
    Index 0 Param 4=5
    Index 1 Param 1=&FloatFromCurve(
     FloatCurve=DataInterface() {propertyValues {Curve {keys [
         {arriveTangent=-3.7492, interpMode=RCIM_Cubic, leaveTangent=-3.7492, tangentMode=RCTM_User, time=0, value=1}
         {arriveTangent=-0.7477, interpMode=RCIM_Cubic, leaveTangent=-0.7477, tangentMode=RCTM_User, time=0.3, value=0.3}
         {interpMode=RCIM_Cubic, leaveTangent=-1.9286, time=1, value=0}
        ]}}}
     CurveIndex=$Particles.Timeline.NormalizedTime
     Scale Curve=&RandomRangeFloat(Minimum=2, Maximum=2.25)
    )
    Module.Param0WriteEnabled=true
    Module.Param0WriteEnabled1=true
    Module.Param1WriteEnabled=true
    Module.Param1WriteEnabled1=false
    Module.Param2WriteEnabled=true
    Module.Param2WriteEnabled1=false
    Module.Param3WriteEnabled=true
    Module.Param3WriteEnabled1=false
    Write Parameter Index 1=true
   )
   &ScaleColor(Scale RGB=<0 0 0>, ScaleA=false, ScaleRGB=true)
   &ScaleSpriteSize(
    Uniform Curve Scale=&MakeCustomFloatFromBool(
     Bool=$System.SpawnShotgun
     True Float=0.4
     False Float=&MakeCustomFloatFromBool(Bool=$System.SpawnPistol, True Float=0.6, False Float=1)
    )
    Uniform Curve Sprite Scale=DataInterface() {propertyValues {Curve {keys [
        {interpMode=RCIM_Cubic, tangentMode=RCTM_User, time=0.7, value=1}
        {arriveTangent=-6.3247, interpMode=RCIM_Cubic, leaveTangent=-6.3247, tangentMode=RCTM_User, time=1, value=0}
       ]}}}
   )
  ]
  rendererClasses [&NiagaraSpriteRendererProperties]
  renderers [{rendererClass=&NiagaraSpriteRendererProperties}]
 }
 NE_Particles_Type001 {
  emitterUpdateScript [
   &EmitterState(
    Enable Distance Culling=true
    MaxDistance=10000
    Module.bUseMaxDistance=true
    Module.bUseMinDistance=false
    Scalability Mode=::NewEnumerator1
   )
   &SpawnRate(SpawnRate=15, Use Spawn Probability=false) {enabled=false}
   &SpawnBurst_Instantaneous(
    Spawn Count=&Multiply_Int(A=12, B=&MakeCustomIntFromBool(Bool=$System.SpawnMud, True INT=1, False INT=0))
    Spawn Group:NiagaraInt32=$System.NewGroup
    Spawn Time=$System.Delay
    Use Spawn Probability=false
    UseLoopCountLimit=false
   )
  ]
  particleSpawnScript [
   /SetVariables(Particles.VisibilityTag=$Emitter.SpawnGroup) {isSetParameters=true}
   &InitializeParticle(
    Color=<rgba 1.3728 3.0599 5 1>
    Lifetime Min=1.5
    Lifetime Mode=::NewEnumerator1
    Sprite Size Mode=::NewEnumerator3
    Uniform Sprite Size=&Multiply_Float(A=&RandomRangeFloat(Minimum=8, Maximum=10), B=$User.Global Scale)
    UsePositionOffset=false
    Write Material Random=true
   )
   &ShapeLocation(
    Apply To Particle Position=::NewEnumerator1
    MaskSpawnGroup=false
    Offset=<5 0 30>
    Offset Coordinate Space=::World
    Offset Mode=::NewEnumerator0
    Override Seed=false
    Rotation Coordinate Space=::World
    Rotation Mode=::NewEnumerator1
    Shape Origin=&VectorFromFloat(Value=0)
    Sphere Radius=6
    Sphere Surface Distribution=1
    UseOwnerScale=true
   ) {module=ShapeLocation001}
   &SkeletalMeshLocation(
    Bone / Socket Index=8
    Bone Sampling Mode=::NewEnumerator4
    Interpolate Mesh Sampling=false
    MaskSpawnGroup=false
    Sampled Position Offset=$Particles.ShapeLocation001.ShapePosition
    Skeletal Mesh=$User.SkeletalMesh
    Use Sampled Mesh Attribute as Ribbon ID=false
   )
   &AddVelocity(
    Additional Quaternion Rotation=false
    Rotation Mode=::NewEnumerator3
    Velocity=$Particles.ShapeLocation001.ShapeNormal
    Velocity Speed Scale=&Multiply_Float(A=600, B=$User.Global Scale)
   ) {module=AddVelocity001}
   &AddVelocity(Velocity=<0 0 1>, Velocity Speed Scale=&Multiply_Float(A=1000, B=$User.Global Scale))
   &ApplyInitialForces(Apply Rotational Force to Rotational Velocity=false)
   /SetVariables() {isSetParameters=true, module=SetVariables_#2}
  ]
  particleUpdateScript [
   &ParticleState()
   &GravityForce(Gravity=&MakeCustomVectorFromBool(Bool=$Particles.HasCollided, True Vector=<0 0 -500>, False Vector=<1 0 -1960>))
   &Collision(
    Age Colliding Particles=false
    Friction=1
    Friction During a Bounce=1
    Max CPU Trace Length=100
    Module.Control Rest State Via Surface Normal=false
    Module.EnableMaxCollisionCount=false
    Module.Randomize Collision Normal Vector=false
    Particle Radius Scale=0.1
    Restitution=0
   ) {enabled=false, module=Collision001}
   &InheritSourceMovement(
    Applied Position Scale=&VectorFromFloat(Value=&MakeCustomFloatFromBool(Bool=$Particles.HasCollided, True Float=0, False Float=0.9))
    ApplyForce=false
    ApplyPosition=true
    ApplyPositionFalloff=false
   )
   &SolveForcesAndVelocity(Clamp Velocity=false, Limit Acceleration=false)
   &Timeline(
    Additional Output:ENiagara_TimelineOutput=::NewEnumerator1
    Additional Output Mode:ENiagara_TimelineOutputMode=::NewEnumerator1
    Float Curve=DataInterface() {propertyValues {Curve {keys [
        {arriveTangent=9.0725, interpMode=RCIM_Cubic, leaveTangent=9.0725, tangentMode=RCTM_User, time=0, value=0}
        {arriveTangent=-2.5249, interpMode=RCIM_Cubic, leaveTangent=-2.5249, tangentMode=RCTM_User, time=0.1408, value=0.9847}
        {arriveTangent=-0.0405, interpMode=RCIM_Cubic, leaveTangent=-0.0405, tangentMode=RCTM_User, time=1, value=0}
       ]}}}
    Instantaneous Rewind=true
    Play=$Particles.HasCollided
   )
   &DynamicMaterialParameters(
    Index 0 Param 1=$Particles.Timeline.Float
    Index 0 Param 2=5
    Index 0 Param 3=600
    Index 0 Param 4=5
    Index 1 Param 1=&FloatFromCurve(
     FloatCurve=DataInterface() {propertyValues {Curve {keys [
         {arriveTangent=-3.7492, interpMode=RCIM_Cubic, leaveTangent=-3.7492, tangentMode=RCTM_User, time=0, value=1}
         {arriveTangent=-0.7477, interpMode=RCIM_Cubic, leaveTangent=-0.7477, tangentMode=RCTM_User, time=0.3, value=0.3}
         {interpMode=RCIM_Cubic, leaveTangent=-1.9286, time=1, value=0}
        ]}}}
     CurveIndex=$Particles.Timeline.NormalizedTime
     Scale Curve=&RandomRangeFloat(Minimum=2, Maximum=2.25)
    )
    Module.Param0WriteEnabled=true
    Module.Param0WriteEnabled1=true
    Module.Param1WriteEnabled=true
    Module.Param1WriteEnabled1=false
    Module.Param2WriteEnabled=true
    Module.Param2WriteEnabled1=false
    Module.Param3WriteEnabled=true
    Module.Param3WriteEnabled1=false
    Write Parameter Index 1=true
   )
   &ScaleColor(Scale RGB=<0 0 0>, ScaleA=false, ScaleRGB=true)
   &ScaleSpriteSize(Uniform Curve Sprite Scale=DataInterface() {propertyValues {Curve {keys [
        {interpMode=RCIM_Cubic, tangentMode=RCTM_User, time=0.7, value=1}
        {arriveTangent=-6.3247, interpMode=RCIM_Cubic, leaveTangent=-6.3247, tangentMode=RCTM_User, time=1, value=0}
       ]}}})
  ]
  rendererClasses [&NiagaraSpriteRendererProperties]
  renderers [{rendererClass=&NiagaraSpriteRendererProperties}]
 }
 NE_Particles_Type_1 {
  simTarget=GPUComputeSim
  emitterUpdateScript [
   &EmitterState(
    Enable Distance Culling=true
    MaxDistance=12000
    Module.bUseMaxDistance=true
    Module.bUseMinDistance=false
    Scalability Mode=::NewEnumerator1
   )
   &SpawnPerUnit(Module.Use Max Movement Threshold=true, Module.Use Movement Tolerance=false, Use Spawn Probability=false) {enabled=false}
   &SpawnBurst_Instantaneous(
    Spawn Count=&Multiply_Int(
     A=&MakeIntFromFloat(Float=&Multiply_Float(A=20, B=$User.Global Scale), Conversion Mode=::NewEnumerator2)
     B=&MakeCustomIntFromBool(
      Bool=&Boolean_OrOperation(Bool1=$System.SpawnSand, Bool2=&Boolean_OrOperation(Bool1=$System.SpawnMud, Bool2=$System.SpawnSnow))
      True INT=1
      False INT=0
     )
    )
    Spawn Group:NiagaraInt32=$System.NewGroup
    Use Spawn Probability=false
    UseLoopCountLimit=false
   )
   &SpawnRate(SpawnRate=30, Use Spawn Probability=false) {enabled=false}
  ]
  particleSpawnScript [
   /SetVariables(Particles.VisibilityTag=$Emitter.SpawnGroup) {isSetParameters=true}
   &InitializeParticle(
    Lifetime Min=1.4
    Lifetime Mode=::NewEnumerator1
    Position Offset=&Multiply_VectorByFloat(
     Vector=<0 0 70>
     Float=&FloatFromCurve(
      FloatCurve=DataInterface() {propertyValues {Curve {keys [{tangentMode=RCTM_User, time=0, value=0.1}, {tangentMode=RCTM_User, time=1, value=1}]}}}
      CurveIndex=$System.NormalizedLoopAge
     )
    )
    Sprite Rotation Mode=::NewEnumerator2
    Sprite Size Mode=::NewEnumerator3
    Sprite UV Mode=::NewEnumerator3
    Uniform Sprite Size=&SelectFloatFromArray(Direct Array Index=$Emitter.SpawnGroup, Float Selection Array=DataInterface() {propertyValues {FloatData [200,35,350,350]}})
    UsePositionOffset=true
    Write Material Random=true
   )
   /Niagara/Modules/Update/SubUV/V2/SubUVAnimation(
    "Sprite Renderer "=DataInterface() {propertyValues {SpriteRenderer=&self:NE_Particles_Type_0.NiagaraSpriteRendererProperties_0}}
    SubUV Animation Mode:ENiagara_SubUVLookupModeV2=::NewEnumerator2
    UseEndFrame=false
    UseStartFrame=false
   )
   &ShapeLocation(
    Apply To Particle Position=::NewEnumerator1
    Handle Radius:NiagaraFloat=&Multiply_Float(A=210, B=$User.Global Scale)
    Large Radius=&Multiply_Float(A=250, B=$User.Global Scale)
    MaskSpawnGroup=false
    Non Uniform Scale=<1.5 1.5 1>
    Offset=<0 0 10>
    Offset Coordinate Space=::World
    Offset Mode=::NewEnumerator0
    Override Seed=false
    Rotation Axis=<0 0 1>
    Rotation Coordinate Space=::World
    Rotation Mode=::NewEnumerator1
    Shape Primitive=::NewEnumerator3
    UseOwnerScale=true
   ) {module=ShapeLocation001}
   &AddVelocity(Velocity=$Particles.ShapeLocation001.ShapeNormal, Velocity Speed Scale=400) {enabled=false}
   &AddVelocity(
    Cone Angle=90
    Cone Axis=<0 0 1>
    Module.UseConeFalloff=false
    Module.UseVelDistribution=false
    Override Seed=false
    Velocity Mode=::NewEnumerator2
    Velocity Speed=&Multiply_Float(A=&RandomRangeFloat(Minimum=800, Maximum=1000), B=$User.Global Scale)
   ) {module=AddVelocity001}
   &CameraOffset(Camera Offset Amount=15)
   &InitialMeshOrientation(Use Rotation Vector=false)
  ]
  particleUpdateScript [
   &ParticleState()
   &UpdateMeshOrientation()
   &SpriteRotationRate(Rotation Rate=&Multiply_Float(
     A=&RandomRangeFloat(Minimum=20, Maximum=30)
     B=HlslExpression() {hlslExpression="sign(Particles.MaterialRandom-0.5)"}
    ))
   &GravityForce(Gravity=<0 0 -1200>)
   &InheritSourceMovement(
    Applied Position Scale=&VectorFromFloat(Value=0.95)
    ApplyForce=true
    ApplyForceFalloff=false
    ApplyPosition=true
    ApplyPositionFalloff=false
    Module.LimitForceSpeed=false
   )
   &SolveForcesAndVelocity(Clamp Velocity=false, Limit Acceleration=false)
   &ScaleSpriteSize(
    Uniform Curve Scale=&Multiply_Float(A=1, B=$User.Global Scale)
    Uniform Curve Sprite Scale=DataInterface() {propertyValues {Curve {keys [
        {arriveTangent=4.1785, interpMode=RCIM_Cubic, leaveTangent=4.1785, tangentMode=RCTM_User, time=0, value=0.15}
        {arriveTangent=1.1179, interpMode=RCIM_Cubic, leaveTangent=1.1179, tangentMode=RCTM_User, time=0.2042, value=0.6228}
        {arriveTangent=0.0224, interpMode=RCIM_Cubic, leaveTangent=0.0224, tangentMode=RCTM_User, time=1, value=1}
       ]}}}
   )
   &ScaleColor(
    Scale Alpha=&FloatFromCurve(FloatCurve=DataInterface() {propertyValues {Curve {keys [<at 0 1 RCIM_Cubic>, {arriveTangent=-2.0383, interpMode=RCIM_Cubic, leaveTangent=-2.0383, tangentMode=RCTM_User, time=1, value=0}]}}})
    Scale RGB=&SelectVectorFromArray(
     Direct Array Index=$Particles.VisibilityTag
     Vector Selection Array=DataInterface() {propertyValues {FloatData [<0.7 0.7 0.7>, <0.9 0.9 0.9>, <0.8 0.5 0.3>, <0 0 0>]}}
    )
    ScaleA=true
    ScaleRGB=true
   )
  ]
  rendererClasses [&NiagaraSpriteRendererProperties]
  renderers [
   {rendererClass=&NiagaraSpriteRendererProperties}
   {rendererClass=&NiagaraSpriteRendererProperties, rendererIndex=1}
   {rendererClass=&NiagaraSpriteRendererProperties, rendererIndex=2}
  ]
 }
 PlayAudioOnSpawnBurst {
  emitterSpawnScript [/SetVariables(Emitter.AudioPlayer=DataInterface() {propertyValues {
      Attenuation=null
      Concurrency=null
      MaxPlaysPerTick=10
      SoundToPlay=/Game/Sounds/Fort_Weapons/Impacts/Mud/Mud_Explosion_Medium_Cue
      bAllowLoopingOneShotSounds=false
      bLimitPlaysPerTick=true
      bOnlyActiveDuringGameplay=false
      bStopWhenComponentIsDestroyed=true
      ConfigurationUserParameter {parameter {name=None, typeDefHandle {registeredTypeIndex=1}, varData []}}
      ParameterNames []
     }}) {isSetParameters=true}]
  emitterUpdateScript [
   &EmitterState(Life Cycle Mode=::NewEnumerator1, Loop Behavior=::NewEnumerator1, Loop Duration=0.25, UseLoopDelay=false)
   &SpawnBurst_Instantaneous(Use Spawn Probability=false, UseLoopCountLimit=false)
  ]
  particleSpawnScript [
   &InitializeParticle(UsePositionOffset=false, Write Material Random=true)
   /Niagara/Modules/Audio/PlayAudio(
    Audio Player:NiagaraDataInterfaceAudioPlayer=$Emitter.AudioPlayer
    Play Audio Trigger:NiagaraBool=&SetBoolByIntComparison(
     A=&ReturnParticleIDAttributes(Particle ID Attribute=::NewEnumerator1)
     B=0
     Comparison Type=::NewEnumerator4
    )
   )
  ]
  particleUpdateScript [&SolveForcesAndVelocity(Clamp Velocity=false, Limit Acceleration=false) {enabled=false}]
  rendererClasses []
  renderers []
 }
}
summary {
 systemName=NS_Sniper_ExplosiveRepeater_Mud
 emitters [
  {emitterName=NE_Particles_Type_1, simTarget=GPUComputeSim, rendererClasses [&NiagaraSpriteRendererProperties]}
  {emitterName=NE_Particles_Type001, rendererClasses [&NiagaraSpriteRendererProperties]}
  {emitterName=Collision_Send_BP, rendererClasses [&NiagaraSpriteRendererProperties]}
  {
   emitterName=Drops_Wide
   simTarget=GPUComputeSim
   rendererClasses [&NiagaraSpriteRendererProperties, &NiagaraMeshRendererProperties]
  }
  {emitterName=Decal, rendererClasses [&NiagaraDecalRendererProperties]}
  {emitterName=PlayAudioOnSpawnBurst, rendererClasses []}
  {emitterName=Collided_Mud_Splash001, rendererClasses [&NiagaraSpriteRendererProperties]}
  {emitterName=NE_Parent_Mud_DropsCollision, rendererClasses [&NiagaraSpriteRendererProperties]}
 ]
 userVariables [
  User.bIsRight {type=NiagaraBool}
  User.Export particle data {
   type=Object
   description=""
   defaultValue {struct=NiagaraExt_VariableValue_Object, value {object=null, objectClass=&Object}}
  }
  User.FootDirection {type=Vector3f, description="", defaultValue {struct=Vector3f, value=<0 0 0>}}
  User.Global Scale {type=NiagaraFloat, defaultValue=0.6, description=""}
  User.GroundNormal {type=Vector3f, description="", defaultValue {struct=Vector3f, value=<0 0 0>}}
  User.PeelOutColorArray {
   type=NiagaraDataInterfaceArrayColor
   description=""
   defaultValue {
    struct=NiagaraExt_VariableValue_DataInterface
    value {
     dataInterface=&self:NiagaraDataInterfaceArrayColor_0
     dataInterfaceClass=&NiagaraDataInterfaceArrayColor
    }
   }
  }
  User.PeelOutColorIndex {type=NiagaraInt32}
  User.PhysicalSurface {type=NiagaraInt32, defaultValue=30}
  User.SkeletalMesh {
   type=NiagaraDataInterfaceSkeletalMesh
   defaultValue {
    struct=NiagaraExt_VariableValue_DataInterface
    value {dataInterface=&self:NiagaraDataInterfaceSkeletalMesh_0}
   }
  }
 ]
}
