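# Sound cue for the Explosive Repeater's aim-down-sights-in foley.
# A Random node picks between two ADS-in wave variations at equal weight, then a Modulator lowers pitch and slightly varies volume on each play.
# It plays in the weapon foley sound class at reduced volume, using the shared weapon reload attenuation.
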SoundCue /ChronoWeaponGameplay/Items/ExplosiveRepeater/Sounds/ExplosiveRepeater_ADS_In_Cue
root=Modulator_0
nodes [
 {id=Modulator_0, children [Random_0], properties {pitchMax=0.9, pitchMin=0.8, volumeMax=1.1, volumeMin=0.9}}
 {id=Random_0, children [WavePlayer_4,WavePlayer_5], properties {weights [1,1]}}
 {id=WavePlayer_4, properties {soundWaveAssetPtr=&/ExplosiveRepeater_ADS_In_01}}
 {id=WavePlayer_5, properties {soundWaveAssetPtr=&/ExplosiveRepeater_ADS_In_02}}
]
properties {
 attenuationSettings=&Fort_Weapon_Reload_Att
 maxDistance=2250
 soundClassObject=&Fort_WeaponFoley
 volumeMultiplier=0.351
}
