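# Sound cue for the Explosive Repeater's aim-down-sights exit foley.
# It randomly picks one of two equally weighted wave variants, then a modulator varies volume slightly and shifts pitch down within a small random range.
# It plays at reduced volume through the weapon foley sound class, using the shared weapon reload attenuation.
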SoundCue /ChronoWeaponGameplay/Items/ExplosiveRepeater/Sounds/ExplosiveRepeater_ADS_Out_Cue
root=Modulator_0
nodes [
 {id=Modulator_0, children [Random_0], properties {pitchMax=0.9, pitchMin=0.8, volumeMax=1.1, volumeMin=0.9}}
 {id=Random_0, children [WavePlayer_1,WavePlayer_2], properties {weights [1,1]}}
 {id=WavePlayer_1, properties {soundWaveAssetPtr=&/ExplosiveRepeater_ADS_Out_01}}
 {id=WavePlayer_2, properties {soundWaveAssetPtr=&/ExplosiveRepeater_ADS_Out_02}}
]
properties {
 attenuationSettings=&Fort_Weapon_Reload_Att
 maxDistance=2250
 soundClassObject=&Fort_WeaponFoley
 volumeMultiplier=0.351
}
