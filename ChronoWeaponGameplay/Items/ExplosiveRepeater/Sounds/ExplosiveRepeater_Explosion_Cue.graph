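# Sound cue for the Explosive Repeater's explosion, played after a short fixed delay.
# A quality switch selects five variations per layer for its first two levels and three for its last two.
# Each branch mixes random close, medium and distant waves, each with its own attenuation and modulation, the distant layer pitched down.
# Overridden concurrency caps instances per owner, stopping the oldest.
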SoundCue /ChronoWeaponGameplay/Items/ExplosiveRepeater/Sounds/ExplosiveRepeater_Explosion_Cue
root=Delay_0
nodes [
 {id=Attenuation_0, children [Random_2], properties {attenuationSettings=/Game/Sounds/Attenuation/Fort_Explosion_Medium_Athena_Att}}
 {id=Attenuation_1, children [Random_3], properties {attenuationSettings=/Game/Sounds/Attenuation/Fort_Explosion_Distant_Athena_Att}}
 {
  id=Attenuation_5
  children [Modulator_2]
  properties {attenuationSettings=/Game/Sounds/Attenuation/Fort_Explosion_Close_Athena_Att, attenuationOverrides {falloffDistance=1600}}
 }
 {id=Attenuation_6, children [Random_5], properties {attenuationSettings=&Fort_Explosion_Medium_Athena_Att}}
 {
  id=Attenuation_7
  children [Modulator_3]
  properties {attenuationSettings=&Fort_Explosion_Close_Athena_Att, attenuationOverrides {falloffDistance=1600}}
 }
 {id=Attenuation_8, children [Random_6], properties {attenuationSettings=&Fort_Explosion_Distant_Athena_Att}}
 {id=Delay_0, children [QualityLevel_1], properties {delayMax=0.08, delayMin=0.08}}
 {id=Mixer_0, children [Attenuation_7, Modulator_1, Modulator_0], properties {inputVolume [1.1,0.9,0.4]}}
 {id=Mixer_1, children [Attenuation_5, Modulator_6, Modulator_5], properties {inputVolume [1.1,0.9,0.4]}}
 {id=Modulator_0, children [Attenuation_1], properties {pitchMax=0.7, pitchMin=0.7, volumeMax=1, volumeMin=0.9}}
 {id=Modulator_1, children [Attenuation_0], properties {pitchMax=1.1, pitchMin=0.9, volumeMax=1.1, volumeMin=0.9}}
 {id=Modulator_2, children [Random_0], properties {pitchMax=1.1, pitchMin=0.9, volumeMax=1.1, volumeMin=0.9}}
 {id=Modulator_3, children [Random_1], properties {pitchMax=1.1, pitchMin=0.9, volumeMax=1.1, volumeMin=0.9}}
 {id=Modulator_5, children [Attenuation_8], properties {pitchMax=0.7, pitchMin=0.7, volumeMax=1, volumeMin=0.9}}
 {id=Modulator_6, children [Attenuation_6], properties {pitchMax=1.1, pitchMin=0.9, volumeMax=1.1, volumeMin=0.9}}
 {id=QualityLevel_1, children [Mixer_0,Mixer_0,Mixer_1,Mixer_1]}
 {id=Random_0, children [WavePlayer_0,WavePlayer_1,WavePlayer_2], properties {weights [1,1,1]}}
 {id=Random_1, children [WavePlayer_4,WavePlayer_5,WavePlayer_6,WavePlayer_7,WavePlayer_8], properties {weights [1,1,1,1,1]}}
 {id=Random_2, children [WavePlayer_9, WavePlayer_10, WavePlayer_11, WavePlayer_12, WavePlayer_13], properties {weights [1,1,1,1,1]}}
 {id=Random_3, children [WavePlayer_26, WavePlayer_27, WavePlayer_28, WavePlayer_29, WavePlayer_30], properties {weights [1,1,1,1,1]}}
 {id=Random_5, children [WavePlayer_15, WavePlayer_16, WavePlayer_17], properties {weights [1,1,1]}}
 {id=Random_6, children [WavePlayer_20, WavePlayer_21, WavePlayer_22], properties {weights [1,1,1]}}
 {id=WavePlayer_0, properties {soundWaveAssetPtr=&/ExplosiveRepeater_Explosion_Close_01}}
 {id=WavePlayer_1, properties {soundWaveAssetPtr=&/ExplosiveRepeater_Explosion_Close_02}}
 {id=WavePlayer_10, properties {soundWaveAssetPtr=&/ExplosiveRepeater_Explosion_Medium_02}}
 {id=WavePlayer_11, properties {soundWaveAssetPtr=&/ExplosiveRepeater_Explosion_Medium_03}}
 {id=WavePlayer_12, properties {soundWaveAssetPtr=&/ExplosiveRepeater_Explosion_Medium_04}}
 {id=WavePlayer_13, properties {soundWaveAssetPtr=&/ExplosiveRepeater_Explosion_Medium_05}}
 {id=WavePlayer_15, properties {soundWaveAssetPtr=&/ExplosiveRepeater_Explosion_Medium_01}}
 {id=WavePlayer_16, properties {soundWaveAssetPtr=&ExplosiveRepeater_Explosion_Medium_02}}
 {id=WavePlayer_17, properties {soundWaveAssetPtr=&ExplosiveRepeater_Explosion_Medium_03}}
 {id=WavePlayer_2, properties {soundWaveAssetPtr=&/ExplosiveRepeater_Explosion_Close_03}}
 {id=WavePlayer_20, properties {soundWaveAssetPtr=&/ExplosiveRepeater_Explosion_Distant_01}}
 {id=WavePlayer_21, properties {soundWaveAssetPtr=&/ExplosiveRepeater_Explosion_Distant_02}}
 {id=WavePlayer_22, properties {soundWaveAssetPtr=&/ExplosiveRepeater_Explosion_Distant_03}}
 {id=WavePlayer_26, properties {soundWaveAssetPtr=&ExplosiveRepeater_Explosion_Distant_01}}
 {id=WavePlayer_27, properties {soundWaveAssetPtr=&ExplosiveRepeater_Explosion_Distant_02}}
 {id=WavePlayer_28, properties {soundWaveAssetPtr=&ExplosiveRepeater_Explosion_Distant_03}}
 {id=WavePlayer_29, properties {soundWaveAssetPtr=&/ExplosiveRepeater_Explosion_Distant_04}}
 {id=WavePlayer_30, properties {soundWaveAssetPtr=&/ExplosiveRepeater_Explosion_Distant_05}}
 {id=WavePlayer_4, properties {soundWaveAssetPtr=&ExplosiveRepeater_Explosion_Close_01}}
 {id=WavePlayer_5, properties {soundWaveAssetPtr=&ExplosiveRepeater_Explosion_Close_02}}
 {id=WavePlayer_6, properties {soundWaveAssetPtr=&ExplosiveRepeater_Explosion_Close_03}}
 {id=WavePlayer_7, properties {soundWaveAssetPtr=&/ExplosiveRepeater_Explosion_Close_04}}
 {id=WavePlayer_8, properties {soundWaveAssetPtr=&/ExplosiveRepeater_Explosion_Close_05}}
 {id=WavePlayer_9, properties {soundWaveAssetPtr=&ExplosiveRepeater_Explosion_Medium_01}}
]
properties {
 bExcludeFromRandomNodeBranchCulling=true
 bOverrideConcurrency=true
 maxDistance=27500
 soundClassObject=/Game/Sounds/FortSoundClassesAndModes/Fort_Explosives_Large_PSM
 volumeMultiplier=0.9
 concurrencyOverrides {bLimitToOwner=true, maxCount=3, resolutionRule=StopOldest, volumeScaleAttackTime=0.1}
}
