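# Sound cue for the first-person firing sound of the Explosive Repeater weapon.
# A quality-level node gives its first two slots a full mix of four random close shots, a delayed sub punch and a later random casing sound.
# The other two slots use a lighter mix of two shots and the punch.
# Modulators vary pitch, and concurrency limits voices per owner, stopping the oldest.
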SoundCue /ChronoWeaponGameplay/Items/ExplosiveRepeater/Sounds/ExplosiveRepeater_Fire_1P_Cue
root=QualityLevel_0
nodes [
 {id=Delay_1, children [WavePlayer_8], properties {delayMax=0.1, delayMin=0.1}}
 {id=Delay_2, children [WavePlayer_21], properties {delayMax=0.1, delayMin=0.1}}
 {id=Delay_3, children [Random_3], properties {delayMax=0.7, delayMin=0.6}}
 {id=Mixer_1, children [Random_0,Delay_1,Modulator_2], properties {inputVolume [1.1,0.35,0.25]}}
 {id=Mixer_2, children [Random_2,Delay_2], properties {inputVolume [1,0.35]}}
 {id=Modulator_0, children [Mixer_1], properties {pitchMax=1.1, pitchMin=0.9, volumeMax=1.1, volumeMin=1}}
 {id=Modulator_1, children [Mixer_2], properties {pitchMax=1.1, pitchMin=0.9}}
 {id=Modulator_2, children [Delay_3], properties {pitchMax=1.1, pitchMin=0.9, volumeMax=1.1, volumeMin=0.9}}
 {id=QualityLevel_0, children [Modulator_0,Modulator_0,Modulator_1,Modulator_1]}
 {id=Random_0, children [WavePlayer_0,WavePlayer_1,WavePlayer_6,WavePlayer_7], properties {weights [1,1,1,1]}}
 {id=Random_2, children [WavePlayer_9, WavePlayer_17], properties {weights [1,1]}}
 {id=Random_3, children [WavePlayer_11, WavePlayer_12, WavePlayer_13], properties {weights [1,1,1]}}
 {id=WavePlayer_0, properties {soundWaveAssetPtr=&/ExplosiveRepeater_Fire_Close_01}}
 {id=WavePlayer_1, properties {soundWaveAssetPtr=&/ExplosiveRepeater_Fire_Close_02}}
 {id=WavePlayer_11, properties {soundWaveAssetPtr=/Game/Sounds/Fort_Weapons/Named_Weapons/CowBoy_Repeater/WPN_CowboyRepeater_Casing_01}}
 {id=WavePlayer_12, properties {soundWaveAssetPtr=/Game/Sounds/Fort_Weapons/Named_Weapons/CowBoy_Repeater/WPN_CowboyRepeater_Casing_02}}
 {id=WavePlayer_13, properties {soundWaveAssetPtr=/Game/Sounds/Fort_Weapons/Named_Weapons/CowBoy_Repeater/WPN_CowboyRepeater_Casing_03}}
 {id=WavePlayer_17, properties {soundWaveAssetPtr=&ExplosiveRepeater_Fire_Close_02}}
 {id=WavePlayer_21, properties {soundWaveAssetPtr=/Game/Sounds/Fort_Weapons/Shotgun/Charge/Sub_Punch_01}}
 {id=WavePlayer_6, properties {soundWaveAssetPtr=&/ExplosiveRepeater_Fire_Close_03}}
 {id=WavePlayer_7, properties {soundWaveAssetPtr=&/ExplosiveRepeater_Fire_Close_04}}
 {id=WavePlayer_8, properties {soundWaveAssetPtr=&Sub_Punch_01}}
 {id=WavePlayer_9, properties {soundWaveAssetPtr=&ExplosiveRepeater_Fire_Close_01}}
]
properties {
 attenuationSettings=&Fort_Weapon_Projectile_FirstPerson_Athena_Att
 bExcludeFromRandomNodeBranchCulling=true
 maxDistance=2097152
 soundClassObject=&Fort_WeaponRanged_Fire_1P_PSM
 volumeMultiplier=1.5
 attenuationOverrides {distanceAlgorithm=NaturalSound, falloffDistance=4000, attenuationShapeExtents {x=1000}}
 concurrencyOverrides {bLimitToOwner=true, maxCount=3, resolutionRule=StopOldest}
 concurrencySet [&Weapon_Gunfire_FirstPerson_Conc]
}
