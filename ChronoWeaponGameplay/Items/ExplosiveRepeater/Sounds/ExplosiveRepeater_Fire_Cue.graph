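# Sound cue for the Explosive Repeater's gunfire, layering close, medium and distant shots with close weighted loudest.
# Each layer runs random waves through a modulator, its own projectile attenuation and a sound class: third-person fire for close, distant fire for the others.
# A quality switch at the root picks between two such mixes, one with four variations per layer and a lighter one with three.
# Global and local gunfire concurrency sets limit overlapping shots.
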SoundCue /ChronoWeaponGameplay/Items/ExplosiveRepeater/Sounds/ExplosiveRepeater_Fire_Cue
root=QualityLevel_2
nodes [
 {id=Attenuation_11, children [Modulator_0], properties {attenuationSettings=&Fort_Weapon_Projectile_Close_Athena_Att}}
 {
  id=Attenuation_12
  children [Modulator_6]
  properties {attenuationSettings=&Fort_Weapon_Projectile_Medium_Mono_Athena_Att}
 }
 {
  id=Attenuation_13
  children [Modulator_7]
  properties {attenuationSettings=&Fort_Weapon_Projectile_Distant_Mono_Athena_Att}
 }
 {id=Attenuation_6, children [Modulator_11], properties {attenuationSettings=&Fort_Weapon_Projectile_Close_Athena_Att}}
 {
  id=Attenuation_7
  children [Modulator_12]
  properties {attenuationSettings=&Fort_Weapon_Projectile_Medium_Mono_Athena_Att}
 }
 {
  id=Attenuation_8
  children [Modulator_13]
  properties {attenuationSettings=&Fort_Weapon_Projectile_Distant_Mono_Athena_Att}
 }
 {id=Mixer_1, children [SoundClass_14, SoundClass_16, SoundClass_15], properties {inputVolume [1.2,0.5,0.35]}}
 {id=Mixer_3, children [SoundClass_6,SoundClass_8,SoundClass_7], properties {inputVolume [1.2,0.7,0.35]}}
 {id=Modulator_0, children [Random_3]}
 {id=Modulator_11, children [Random_11]}
 {id=Modulator_12, children [Random_12]}
 {id=Modulator_13, children [Random_13]}
 {id=Modulator_6, children [Random_0]}
 {id=Modulator_7, children [Random_1]}
 {id=QualityLevel_2, children [Mixer_1,Mixer_1,Mixer_3,Mixer_3]}
 {id=Random_0, children [WavePlayer_0,WavePlayer_1,WavePlayer_2,WavePlayer_3], properties {weights [1,1,1,1]}}
 {id=Random_1, children [WavePlayer_4,WavePlayer_5,WavePlayer_6,WavePlayer_7], properties {weights [1,1,1,1]}}
 {id=Random_11, children [WavePlayer_43, WavePlayer_44, WavePlayer_45], properties {weights [1,1,1]}}
 {id=Random_12, children [WavePlayer_47, WavePlayer_48, WavePlayer_49], properties {weights [1,1,1]}}
 {id=Random_13, children [WavePlayer_51, WavePlayer_52, WavePlayer_53], properties {weights [1,1,1]}}
 {id=Random_3, children [WavePlayer_8, WavePlayer_9, WavePlayer_10, WavePlayer_11], properties {weights [1,1,1,1]}}
 {
  id=SoundClass_14
  children [Attenuation_11]
  properties {soundClassOverride=&Fort_WeaponRanged_Fire_3P_PSM}
 }
 {
  id=SoundClass_15
  children [Attenuation_13]
  properties {soundClassOverride=&Fort_WeaponRanged_Fire_Distant}
 }
 {
  id=SoundClass_16
  children [Attenuation_12]
  properties {soundClassOverride=&Fort_WeaponRanged_Fire_Distant}
 }
 {
  id=SoundClass_6
  children [Attenuation_6]
  properties {soundClassOverride=&Fort_WeaponRanged_Fire_3P_PSM}
 }
 {
  id=SoundClass_7
  children [Attenuation_8]
  properties {soundClassOverride=&Fort_WeaponRanged_Fire_Distant}
 }
 {
  id=SoundClass_8
  children [Attenuation_7]
  properties {soundClassOverride=&Fort_WeaponRanged_Fire_Distant}
 }
 {id=WavePlayer_0, properties {soundWaveAssetPtr=&/ExplosiveRepeater_Fire_Medium_01}}
 {id=WavePlayer_1, properties {soundWaveAssetPtr=&/ExplosiveRepeater_Fire_Medium_02}}
 {id=WavePlayer_10, properties {soundWaveAssetPtr=&/ExplosiveRepeater_Fire_Close_03}}
 {id=WavePlayer_11, properties {soundWaveAssetPtr=&/ExplosiveRepeater_Fire_Close_04}}
 {id=WavePlayer_2, properties {soundWaveAssetPtr=&/ExplosiveRepeater_Fire_Medium_03}}
 {id=WavePlayer_3, properties {soundWaveAssetPtr=&/ExplosiveRepeater_Fire_Medium_04}}
 {id=WavePlayer_4, properties {soundWaveAssetPtr=&/ExplosiveRepeater_Fire_Distant_01}}
 {id=WavePlayer_43, properties {soundWaveAssetPtr=&/ExplosiveRepeater_Fire_Close_01}}
 {id=WavePlayer_44, properties {soundWaveAssetPtr=&/ExplosiveRepeater_Fire_Close_02}}
 {id=WavePlayer_45, properties {soundWaveAssetPtr=&ExplosiveRepeater_Fire_Close_03}}
 {id=WavePlayer_47, properties {soundWaveAssetPtr=&ExplosiveRepeater_Fire_Medium_01}}
 {id=WavePlayer_48, properties {soundWaveAssetPtr=&ExplosiveRepeater_Fire_Medium_02}}
 {id=WavePlayer_49, properties {soundWaveAssetPtr=&ExplosiveRepeater_Fire_Medium_03}}
 {id=WavePlayer_5, properties {soundWaveAssetPtr=&/ExplosiveRepeater_Fire_Distant_02}}
 {id=WavePlayer_51, properties {soundWaveAssetPtr=&ExplosiveRepeater_Fire_Distant_01}}
 {id=WavePlayer_52, properties {soundWaveAssetPtr=&ExplosiveRepeater_Fire_Distant_02}}
 {id=WavePlayer_53, properties {soundWaveAssetPtr=&/ExplosiveRepeater_Fire_Distant_03}}
 {id=WavePlayer_6, properties {soundWaveAssetPtr=&ExplosiveRepeater_Fire_Distant_03}}
 {id=WavePlayer_7, properties {soundWaveAssetPtr=&/ExplosiveRepeater_Fire_Distant_04}}
 {id=WavePlayer_8, properties {soundWaveAssetPtr=&ExplosiveRepeater_Fire_Close_01}}
 {id=WavePlayer_9, properties {soundWaveAssetPtr=&ExplosiveRepeater_Fire_Close_02}}
]
properties {
 bExcludeFromRandomNodeBranchCulling=true
 maxDistance=27500
 volumeMultiplier=1.3
 attenuationOverrides {distanceAlgorithm=NaturalSound, falloffDistance=4000, attenuationShapeExtents {x=1000}}
 concurrencyOverrides {bLimitToOwner=true, maxCount=3, resolutionRule=StopOldest}
 concurrencySet [&Weapon_Gunfire_Global_Conc, &Weapon_Gunfire_Local_Conc]
}
