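# Sound cue for the looping in-air sound of an Explosive Repeater projectile.
# A random node picks one of three equally weighted looping waves, and a modulator varies its pitch slightly and boosts its volume.
# It uses a shared projectile in-air attenuation and a third-person ranged weapon fire sound class.
# A concurrency override caps simultaneous instances to limit stacking when several projectiles are airborne.
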SoundCue /ChronoWeaponGameplay/Items/ExplosiveRepeater/Sounds/ExplosiveRepeater_InAirLoop_Cue
root=Modulator_1
nodes [
 {id=Modulator_1, children [Random_0], properties {pitchMax=1.1, pitchMin=0.9, volumeMax=1.2, volumeMin=1.1}}
 {id=Random_0, children [WavePlayer_0,WavePlayer_2,WavePlayer_3], properties {weights [1,1,1]}}
 {
  id=WavePlayer_0
  properties {bLooping=true, soundWaveAssetPtr=&/ExplosiveRepeater_InAirLoop_A_01}
 }
 {
  id=WavePlayer_2
  properties {bLooping=true, soundWaveAssetPtr=&/ExplosiveRepeater_InAirLoop_A_02}
 }
 {
  id=WavePlayer_3
  properties {bLooping=true, soundWaveAssetPtr=&/ExplosiveRepeater_InAirLoop_A_03}
 }
]
properties {
 attenuationSettings=/Game/Sounds/Attenuation/Fort_Weapon_Projectile_InAir_Loop_Att
 bOverrideConcurrency=true
 maxDistance=4500
 soundClassObject=&Fort_WeaponRanged_Fire_3P_PSM
 volumeMultiplier=0.4
 concurrencyOverrides {maxCount=3}
}
