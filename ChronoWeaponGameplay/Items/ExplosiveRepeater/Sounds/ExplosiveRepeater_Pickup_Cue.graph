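# Sound cue for picking up the Explosive Repeater weapon.
# A mixer layers a borrowed RatRod sniper pickup wave with a quieter Explosive Repeater pickup wave, each through its own modulator for random pitch and volume, slightly wider on the Repeater layer.
# It uses the loot-collect sound class and weapon-pickup attenuation, with a concurrency override that allows one instance and blocks new ones while it plays.
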SoundCue /ChronoWeaponGameplay/Items/ExplosiveRepeater/Sounds/ExplosiveRepeater_Pickup_Cue
root=Mixer_0
nodes [
 {id=Mixer_0, children [Modulator_0,Modulator_1], properties {inputVolume [1,0.7]}}
 {id=Modulator_0, children [WavePlayer_2], properties {pitchMax=1.1, pitchMin=0.9, volumeMax=1.1, volumeMin=0.9}}
 {id=Modulator_1, children [WavePlayer_0], properties {pitchMax=1.15, pitchMin=0.85, volumeMax=1.15, volumeMin=0.85}}
 {id=WavePlayer_0, properties {soundWaveAssetPtr=&/ExplosiveRepeater_Pickup_01}}
 {id=WavePlayer_2, properties {soundWaveAssetPtr=/Game/Sounds/Fort_Weapons/RatRod/Sniper/RR_Sniper_Pickup_01}}
]
properties {
 attenuationSettings=&Fort_Weapon_Pickup_Att
 bOverrideConcurrency=true
 maxDistance=3500
 soundClassObject=&Fort_Loot_Collect
 volumeMultiplier=0.3877
 concurrencyOverrides {maxCount=1, resolutionRule=PreventNew}
}
