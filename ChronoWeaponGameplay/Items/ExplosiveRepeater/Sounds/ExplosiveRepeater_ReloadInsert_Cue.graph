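# Sound cue for the ExplosiveRepeater weapon's reload insert foley.
# A Random node picks among five equally weighted insert wave variations and feeds a root Modulator node.
# The cue uses the shared weapon reload attenuation and weapon foley sound class, with reduced volume and overridden concurrency.
# Its debug flag is left enabled.
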SoundCue /ChronoWeaponGameplay/Items/ExplosiveRepeater/Sounds/ExplosiveRepeater_ReloadInsert_Cue
root=Modulator_0
nodes [
 {id=Modulator_0, children [Random_0]}
 {id=Random_0, children [WavePlayer_0,WavePlayer_1,WavePlayer_2,WavePlayer_3,WavePlayer_4], properties {weights [1,1,1,1,1]}}
 {id=WavePlayer_0, properties {soundWaveAssetPtr=&/ExplosiveRepeater_ReloadInsert_01}}
 {id=WavePlayer_1, properties {soundWaveAssetPtr=&/ExplosiveRepeater_ReloadInsert_02}}
 {id=WavePlayer_2, properties {soundWaveAssetPtr=&/ExplosiveRepeater_ReloadInsert_03}}
 {id=WavePlayer_3, properties {soundWaveAssetPtr=&/ExplosiveRepeater_ReloadInsert_04}}
 {id=WavePlayer_4, properties {soundWaveAssetPtr=&/ExplosiveRepeater_ReloadInsert_05}}
]
properties {
 attenuationSettings=&Fort_Weapon_Reload_Att
 bDebug=true
 bOverrideConcurrency=true
 maxDistance=2250
 soundClassObject=&Fort_WeaponFoley
 volumeMultiplier=0.501
}
