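# Sound cue for the Explosive Repeater's reload start, using weapon reload attenuation and the weapon foley sound class.
# A mixer layers a random pick of three reload-start waves, varied slightly in pitch and volume, with a quieter second layer.
# That layer reuses one of four reload-end waves, pitched down and fired after a fixed short delay.
# The cue overrides concurrency and has its debug flag enabled.
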SoundCue /ChronoWeaponGameplay/Items/ExplosiveRepeater/Sounds/ExplosiveRepeater_ReloadStart_Cue
root=Mixer_0
nodes [
 {id=Delay_0, children [Random_2], properties {delayMax=0.27, delayMin=0.27}}
 {id=Mixer_0, children [Modulator_0,Modulator_1], properties {inputVolume [1,0.7]}}
 {id=Modulator_0, children [Random_0], properties {pitchMax=1.1, pitchMin=0.9, volumeMax=1.1, volumeMin=0.9}}
 {id=Modulator_1, children [Delay_0], properties {pitchMax=0.9, pitchMin=0.9}}
 {id=Random_0, children [WavePlayer_0,WavePlayer_1,WavePlayer_2], properties {weights [1,1,1]}}
 {id=Random_2, children [WavePlayer_5,WavePlayer_6,WavePlayer_7,WavePlayer_8], properties {weights [1,1,1,1]}}
 {id=WavePlayer_0, properties {soundWaveAssetPtr=&/ExplosiveRepeater_ReloadStart_A_01}}
 {id=WavePlayer_1, properties {soundWaveAssetPtr=&/ExplosiveRepeater_ReloadStart_A_02}}
 {id=WavePlayer_2, properties {soundWaveAssetPtr=&/ExplosiveRepeater_ReloadStart_A_03}}
 {id=WavePlayer_5, properties {soundWaveAssetPtr=&/ExplosiveRepeater_ReloadEnd_Layer_01}}
 {id=WavePlayer_6, properties {soundWaveAssetPtr=&/ExplosiveRepeater_ReloadEnd_Layer_02}}
 {id=WavePlayer_7, properties {soundWaveAssetPtr=&/ExplosiveRepeater_ReloadEnd_Layer_03}}
 {id=WavePlayer_8, properties {soundWaveAssetPtr=&/ExplosiveRepeater_ReloadEnd_Layer_04}}
]
properties {
 attenuationSettings=&Fort_Weapon_Reload_Att
 bDebug=true
 bOverrideConcurrency=true
 maxDistance=2250
 soundClassObject=&Fort_WeaponFoley
 volumeMultiplier=0.476
}
