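# Sound Cue for the Explosive Repeater's close-range whip sound, classed as a ranged weapon projectile effect.
# A Random node picks evenly among four whip waves, then a Modulator varies volume and raises pitch by a random amount.
# Spatial behavior comes from a shared compound bow whoosh-by attenuation asset, as the cue's inline attenuation values are not switched on.
# A concurrency override caps simultaneous instances per owner, and the debug flag is left on.
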SoundCue /ChronoWeaponGameplay/Items/ExplosiveRepeater/Sounds/ExplosiveRepeater_Whip_Close_Cue
root=Modulator_1
nodes [
 {id=Modulator_1, children [Random_1], properties {pitchMax=1.2, pitchMin=1, volumeMax=1.1, volumeMin=0.9}}
 {id=Random_1, children [WavePlayer_0,WavePlayer_5,WavePlayer_6,WavePlayer_7], properties {weights [1,1,1,1]}}
 {id=WavePlayer_0, properties {soundWaveAssetPtr=&/ExplosiveRepeater_Whip_01}}
 {id=WavePlayer_5, properties {soundWaveAssetPtr=&/ExplosiveRepeater_Whip_02}}
 {id=WavePlayer_6, properties {soundWaveAssetPtr=&/ExplosiveRepeater_Whip_03}}
 {id=WavePlayer_7, properties {soundWaveAssetPtr=&/ExplosiveRepeater_Whip_04}}
]
properties {
 attenuationSettings=/Game/Sounds/Fort_Weapons/Explosive_Bow/Weapon_Compound_Bow_Whoosh_By_Att
 bDebug=true
 bOverrideConcurrency=true
 maxDistance=2097152
 soundClassObject=&Fort_WeaponRanged_Projectile_Effects
 volumeMultiplier=1.2
 attenuationOverrides {
  bAttenuate=false
  bEnableOcclusion=true
  bEnableReverbSend=false
  falloffDistance=0
  occlusionLowPassFilterFrequency=2000
  occlusionTraceChannel=ECC_GameTraceChannel11
  occlusionVolumeAttenuation=0
  spatializationAlgorithm=SPATIALIZATION_HRTF
  attenuationShapeExtents {x=50000}
 }
 concurrencyOverrides {bLimitToOwner=true, maxCount=4}
}
