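# Sound cue for the medium variant of the Explosive Repeater's whip sound, routed to the ranged weapon projectile effects sound class.
# A Random node picks one of four equally weighted whip waves and feeds a Modulator that varies pitch within a lowered range.
# Spatialization references a compound bow whoosh-by attenuation asset; the cue also stores its own attenuation values, overrides concurrency and has its debug flag on.
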SoundCue /ChronoWeaponGameplay/Items/ExplosiveRepeater/Sounds/ExplosiveRepeater_Whip_Medium_Cue
root=Modulator_0
nodes [
 {id=Modulator_0, children [Random_0], properties {pitchMax=0.8, pitchMin=0.7}}
 {id=Random_0, children [WavePlayer_1,WavePlayer_2,WavePlayer_3,WavePlayer_4], properties {weights [1,1,1,1]}}
 {id=WavePlayer_1, properties {soundWaveAssetPtr=&/ExplosiveRepeater_Whip_01}}
 {id=WavePlayer_2, properties {soundWaveAssetPtr=&/ExplosiveRepeater_Whip_02}}
 {id=WavePlayer_3, properties {soundWaveAssetPtr=&/ExplosiveRepeater_Whip_03}}
 {id=WavePlayer_4, properties {soundWaveAssetPtr=&/ExplosiveRepeater_Whip_04}}
]
properties {
 attenuationSettings=/Game/Sounds/Fort_Weapons/Explosive_Bow/Weapon_Compound_Bow_Whoosh_By_Att
 bDebug=true
 bOverrideConcurrency=true
 maxDistance=2097152
 soundClassObject=&Fort_WeaponRanged_Projectile_Effects
 volumeMultiplier=1
 attenuationOverrides {
  bAttenuate=false
  bEnableOcclusion=true
  bEnableReverbSend=false
  bUseComplexCollisionForOcclusion=true
  falloffDistance=1000
  occlusionTraceChannel=ECC_GameTraceChannel11
  occlusionVolumeAttenuation=0
  attenuationShapeExtents {x=50000}
 }
}
