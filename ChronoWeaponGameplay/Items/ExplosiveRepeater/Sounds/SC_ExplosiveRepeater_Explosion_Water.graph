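# Sound cue for the Explosive Repeater's explosion in water.
# After a short fixed delay, it mixes close, medium and distant layers, each with its own Fortnite attenuation and volume/pitch modulation, the distant one pitched down.
# Each layer blends a random explosion wave with random water debris: waves up close, distant bursts for the other two.
# A concurrency override caps instances per owner, stopping the oldest.
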SoundCue /ChronoWeaponGameplay/Items/ExplosiveRepeater/Sounds/SC_ExplosiveRepeater_Explosion_Water
root=Delay_0
nodes [
 {id=Attenuation_0, children [Mixer_3], properties {attenuationSettings=/Game/Sounds/Attenuation/Fort_Explosion_Medium_Athena_Att}}
 {id=Attenuation_1, children [Mixer_4], properties {attenuationSettings=/Game/Sounds/Attenuation/Fort_Explosion_Distant_Athena_Att}}
 {
  id=Attenuation_7
  children [Modulator_3]
  properties {attenuationSettings=/Game/Sounds/Attenuation/Fort_Explosion_Close_Athena_Att, attenuationOverrides {falloffDistance=1600}}
 }
 {id=Delay_0, children [Mixer_0], properties {delayMax=0.08, delayMin=0.08}}
 {id=Mixer_0, children [Attenuation_7, Modulator_1, Modulator_0], properties {inputVolume [1.1,0.9,0.4]}}
 {id=Mixer_2, children [Random_1,Random_0], properties {inputVolume [1,0.8]}}
 {id=Mixer_3, children [Random_2,Random_5], properties {inputVolume [1,0.9]}}
 {id=Mixer_4, children [Random_3,Random_4], properties {inputVolume [0.5,1]}}
 {id=Modulator_0, children [Attenuation_1], properties {pitchMax=0.7, pitchMin=0.7, volumeMax=1, volumeMin=0.9}}
 {id=Modulator_1, children [Attenuation_0], properties {pitchMax=1.1, pitchMin=0.9, volumeMax=1.1, volumeMin=0.9}}
 {id=Modulator_3, children [Mixer_2], properties {pitchMax=1.1, pitchMin=0.9, volumeMax=1.1, volumeMin=0.9}}
 {id=Random_0, children [WavePlayer_15, WavePlayer_16, WavePlayer_17, WavePlayer_20], properties {weights [1,1,1,1]}}
 {id=Random_1, children [WavePlayer_4,WavePlayer_5,WavePlayer_6,WavePlayer_7,WavePlayer_8], properties {weights [1,1,1,1,1]}}
 {id=Random_2, children [WavePlayer_9, WavePlayer_10, WavePlayer_11, WavePlayer_12, WavePlayer_13], properties {weights [1,1,1,1,1]}}
 {id=Random_3, children [WavePlayer_26, WavePlayer_27, WavePlayer_28, WavePlayer_29, WavePlayer_30], properties {weights [1,1,1,1,1]}}
 {id=Random_4, children [WavePlayer_3, WavePlayer_14, WavePlayer_18], properties {weights [1,1,1]}}
 {id=Random_5, children [WavePlayer_0,WavePlayer_1,WavePlayer_2], properties {weights [1,1,1]}}
 {
  id=WavePlayer_0
  properties {soundWaveAssetPtr=/ChronoWeaponGameplay/Items/ExplosiveRepeater/Sounds/Soundwaves/EXPL_Debris_Water_Burst_Distant_01}
 }
 {
  id=WavePlayer_1
  properties {soundWaveAssetPtr=/ChronoWeaponGameplay/Items/ExplosiveRepeater/Sounds/Soundwaves/EXPL_Debris_Water_Burst_Distant_02}
 }
 {id=WavePlayer_10, properties {soundWaveAssetPtr=&/ExplosiveRepeater_Explosion_Medium_02}}
 {id=WavePlayer_11, properties {soundWaveAssetPtr=&/ExplosiveRepeater_Explosion_Medium_03}}
 {id=WavePlayer_12, properties {soundWaveAssetPtr=&/ExplosiveRepeater_Explosion_Medium_04}}
 {id=WavePlayer_13, properties {soundWaveAssetPtr=&/ExplosiveRepeater_Explosion_Medium_05}}
 {
  id=WavePlayer_14
  properties {soundWaveAssetPtr=&EXPL_Debris_Water_Burst_Distant_02}
 }
 {id=WavePlayer_15, properties {soundWaveAssetPtr=/ChronoWeaponGameplay/Items/ExplosiveRepeater/Sounds/Soundwaves/EXPL_Debris_Water_Wave_01}}
 {id=WavePlayer_16, properties {soundWaveAssetPtr=/ChronoWeaponGameplay/Items/ExplosiveRepeater/Sounds/Soundwaves/EXPL_Debris_Water_Wave_02}}
 {id=WavePlayer_17, properties {soundWaveAssetPtr=/ChronoWeaponGameplay/Items/ExplosiveRepeater/Sounds/Soundwaves/EXPL_Debris_Water_Wave_03}}
 {
  id=WavePlayer_18
  properties {soundWaveAssetPtr=/ChronoWeaponGameplay/Items/ExplosiveRepeater/Sounds/Soundwaves/EXPL_Debris_Water_Burst_Distant_03}
 }
 {
  id=WavePlayer_2
  properties {soundWaveAssetPtr=&EXPL_Debris_Water_Burst_Distant_03}
 }
 {id=WavePlayer_20, properties {soundWaveAssetPtr=/ChronoWeaponGameplay/Items/ExplosiveRepeater/Sounds/Soundwaves/EXPL_Debris_Water_Wave_04}}
 {id=WavePlayer_26, properties {soundWaveAssetPtr=&/ExplosiveRepeater_Explosion_Distant_01}}
 {id=WavePlayer_27, properties {soundWaveAssetPtr=&/ExplosiveRepeater_Explosion_Distant_02}}
 {id=WavePlayer_28, properties {soundWaveAssetPtr=&/ExplosiveRepeater_Explosion_Distant_03}}
 {id=WavePlayer_29, properties {soundWaveAssetPtr=&/ExplosiveRepeater_Explosion_Distant_04}}
 {
  id=WavePlayer_3
  properties {soundWaveAssetPtr=&EXPL_Debris_Water_Burst_Distant_01}
 }
 {id=WavePlayer_30, properties {soundWaveAssetPtr=&/ExplosiveRepeater_Explosion_Distant_05}}
 {id=WavePlayer_4, properties {soundWaveAssetPtr=&/ExplosiveRepeater_Explosion_Close_01}}
 {id=WavePlayer_5, properties {soundWaveAssetPtr=&/ExplosiveRepeater_Explosion_Close_02}}
 {id=WavePlayer_6, properties {soundWaveAssetPtr=&/ExplosiveRepeater_Explosion_Close_03}}
 {id=WavePlayer_7, properties {soundWaveAssetPtr=&/ExplosiveRepeater_Explosion_Close_04}}
 {id=WavePlayer_8, properties {soundWaveAssetPtr=&/ExplosiveRepeater_Explosion_Close_05}}
 {id=WavePlayer_9, properties {soundWaveAssetPtr=&/ExplosiveRepeater_Explosion_Medium_01}}
]
properties {
 bExcludeFromRandomNodeBranchCulling=true
 bOverrideConcurrency=true
 maxDistance=27500
 soundClassObject=/Game/Sounds/FortSoundClassesAndModes/Fort_Explosives_Large_PSM
 volumeMultiplier=0.9
 concurrencyOverrides {bLimitToOwner=true, maxCount=3, resolutionRule=StopOldest, volumeScaleAttackTime=0.1}
}
